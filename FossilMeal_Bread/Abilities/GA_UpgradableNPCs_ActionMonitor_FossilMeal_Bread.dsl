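; Gameplay ability that monitors actions for the FossilMeal Bread upgradable NPC and reacts when it reaches full upgrade.
; The OnFullUpgrade override calls the parent, then, if the AI pawn's Ability System Component is valid, activates its give-weapon ability by tag and ends this ability.
; The activate and end events are unused parent-call stubs, so the remaining monitoring logic is left to the parent class.
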(event Ability|EventActivateAbility
  (|CallParentFunction) ; This node is disabled and will not be called.\nDrag off pins to build functionality.
)

(event Ability|EventOnEndAbility
  (params (bWasCancelled :type Boolean))
  (|CallParentFunction bWasCancelled) ; This node is disabled and will not be called.\nDrag off pins to build functionality.
)

(fn OnFullUpgrade
  (bind _aipawn (Variables|InternalData|GetAIPawn))
  (bind _returnvalue (Ability|GetAbilitySystemComponent _aipawn))
  (|CallParentFunction)
  (Utilities|IsValid _returnvalue
    (:Is%20Valid
      (Abilities|TryActivateAbilitiesbyTag _returnvalue (Variables|Default|GetTag_AbilityGiveWeapon) false))
    (:Is%20Not%20Valid
      _))
  (Ability|EndAbility))
